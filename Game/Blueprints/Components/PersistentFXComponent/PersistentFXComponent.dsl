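; Actor component that spawns a persistent, replicated FX actor for its owner on the authority, after a random stagger delay when auto-activated.
; It uses a configured actor class or a default mesh actor, sets it dormant with tick off, then notifies via PersistentFXInterface and a dispatcher.
; Owner destruction, or a BuildingActor owner's death or physics awake change, hides the FX actor, flushes dormancy and expires it shortly after.
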(event EventBeginPlay
  (if (Variables|Activation|GetbAutoActivate)
    (Utilities|FlowControl|Delay (* (Math|Random|RandomFloat) 2.0)) ; Random delay to space spawning out a little
    (CallFunction|Initialize)))

(event Custom|initialize
  (Variables|Default|SetcurrentOwner (Components|GetOwner))
  (if (Networking|HasAuthority (Variables|Default|GetcurrentOwner))
    (CallFunction|SpawnFXActor)))

(fn bindAllEventsToFxActor
  (Utilities|FlowControl|Sequence
    (:then_0
      (bind _asbuilding_actor (Utilities|Casting|CastToBuildingActor (Variables|Default|GetcurrentOwner)))
      (EventDispatchers|BindEventtoOnDied _asbuilding_actor (EventDispatchers|CreateEvent :!event handleOwnerDeath))
      (bind _returnvalue (BuildingActor|PhysicsSimulationSupport|GetPhysicsObjectComponent _asbuilding_actor))
      (Utilities|IsValid _returnvalue
        (:Is%20Valid
          (PhysicsObject|BindEventtoOnAwakeChanged _returnvalue (EventDispatchers|CreateEvent :!event handleOwnerAwakeChanged)))))
    (:then_1
      (Game|BindEventtoOnDestroyed (Variables|Default|GetcurrentOwner) (EventDispatchers|CreateEvent :!event handleOwnerDestroyed)))))

(fn handleOwnerAwakeChanged
  (params (SimulatingComponent :type Primitive%20Component%20Object%20Reference) (bIsAwake :type Boolean))
  (CallFunction|KillFxActor))

(fn handleOwnerDeath
  (params (DamagedActor :type Actor%20Object%20Reference) (Damage :type Float%20%28single-precision%29) (InstigatedBy :type Controller%20Object%20Reference) (DamageCauser :type Actor%20Object%20Reference) (HitLocation :type Vector) (FHitComponent :type Primitive%20Component%20Object%20Reference) (BoneName :type Name) (Momentum :type Vector))
  (CallFunction|KillFxActor))

(fn handleOwnerDestroyed
  (params (DestroyedActor :type Actor%20Object%20Reference))
  (CallFunction|KillFxActor))

(fn killFxActor
  (Utilities|IsValid (Variables|Default|GetcurrentFxActor)
    (:Is%20Valid
      (CallFunction|UnbindAllEventstoFxActor)
      (Rendering|SetActorHiddenInGame (Variables|Default|GetcurrentFxActor) true)
      (Networking|FlushNetDormancy (Variables|Default|GetcurrentFxActor))
      (Actor|SetLifeSpan (Variables|Default|GetcurrentFxActor) 5.0)
      (Variables|Default|SetcurrentFxActor 0))))

(fn SpawnFXActor
  (bind _currentfxactor (Variables|Default|GetcurrentFxActor))
  (Utilities|IsValid (Variables|Default|GetcurrentFxActor)
    (:Is%20Not%20Valid
      (Utilities|FlowControl|Sequence
        (:then_0
          (if (Utilities|IsValidClass (Variables|Default|GetfxActorToSpawn))
            (LocalVariables|Setl_actor (Variables|Default|GetfxActorToSpawn))
            (else
              (Utilities|IsValid (Variables|Default|GetfxMeshToSpawn)
                (:Is%20Valid
                  (LocalVariables|Setl_actor "/Game/Blueprints/Components/PersistentFXComponent/Actors/PersistentFX_DefaultMeshActor.PersistentFX_DefaultMeshActor_C"))
                (:Is%20Not%20Valid
                  (return))))))
        (:then_1
          (bind _returnvalue (Game|SpawnActorfromClass (LocalVariables|Getl_actor) (Transformation|GetWorldTransform) "AlwaysSpawn" "SelectDefaultAtRuntime" 0 0))
          (Variables|Default|SetcurrentFxActor _returnvalue)
          (Utilities|FlowControl|Sequence
            (:then_0
              (Variables|/Script/Engine.Actor|SetbAlwaysRelevant true _currentfxactor)
              (Networking|SetReplicates _currentfxactor true)
              (Networking|SetNetDormancy _currentfxactor "DORM_DormantAll")
              (Actor|Tick|SetActorTickEnabled _currentfxactor false)
              (Networking|FlushNetDormancy _currentfxactor))
            (:then_1
              (CallFunction|BindAllEventstoFxActor))
            (:then_2
              (Utilities|IsValid (Variables|Default|GetfxMeshToSpawn)
                (:Is%20Valid
                  (bind _aspersistent_fx_default_mesh_actor (Utilities|Casting|CastToPersistentFX_DefaultMeshActor (Variables|Default|GetcurrentFxActor)))
                  (Class|PersistentFXDefaultMeshActor|SetStaticMesh _aspersistent_fx_default_mesh_actor (Variables|Default|GetfxMeshToSpawn) (Variables|Default|GetfxMeshOverrideMaterials)))))
            (:then_3
              (if (Utilities|DoesObjectImplementInterface (Variables|Default|GetcurrentFxActor) "/Game/Blueprints/Components/PersistentFXComponent/PersistentFXInterface.PersistentFXInterface_C")
                (Class|PersistentFXInterface|SetFxComponentOwner(Message) (Variables|Default|GetcurrentFxActor) self)))
            (:then_4
              (if (Utilities|DoesObjectImplementInterface (Variables|Default|GetcurrentOwner) "/Game/Blueprints/Components/PersistentFXComponent/PersistentFXInterface.PersistentFXInterface_C")
                (Class|PersistentFXInterface|InitPersistentFX(Message) (Variables|Default|GetcurrentOwner) (Variables|Default|GetcurrentFxActor) self)))
            (:then_5
              (Default|CallOnFxActorSpawned (Variables|Default|GetcurrentFxActor) self))))))))

(fn unbindAllEventsToFxActor
  (Utilities|IsValid (Variables|Default|GetcurrentOwner)
    (:Is%20Valid
      (Utilities|FlowControl|Sequence
        (:then_0
          (bind _asbuilding_actor (Utilities|Casting|CastToBuildingActor (Variables|Default|GetcurrentOwner)))
          (LocalVariables|Setl_buildingOwner _asbuilding_actor)
          (Utilities|FlowControl|Sequence
            (:then_0
              (bind _returnvalue (BuildingActor|PhysicsSimulationSupport|GetPhysicsObjectComponent (LocalVariables|Getl_buildingOwner)))
              (Utilities|IsValid _returnvalue
                (:Is%20Valid
                  (PhysicsObject|UnbindEventfromOnAwakeChanged _returnvalue (EventDispatchers|CreateEvent :!event handleOwnerAwakeChanged)))))
            (:then_1
              (EventDispatchers|UnbindEventfromOnDied (LocalVariables|Getl_buildingOwner) (EventDispatchers|CreateEvent :!event handleOwnerDeath)))))
        (:then_1
          (Game|UnbindEventfromOnDestroyed (Variables|Default|GetcurrentOwner) (EventDispatchers|CreateEvent :!event handleOwnerDestroyed)))))))
