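; PersistentFXInterface is a Blueprint interface that defines the contract between PersistentFXComponent and the actors or objects involved in its persistent effects.
; Editor-only functions initialize an editor component from a parent object and spawn an FX actor from a class for a given component.
; Other functions initialize persistent FX with an actor and component, set the owning FX component, and declare a server-only FX actor setup step.
; It holds only signatures, with no logic.
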(fn editorOnly_initEditorComp
  (params (parent :type Object%20Reference)))

(fn editorOnly_SpawnFXActor
  (params (actorClass :type Actor%20Class%20Reference) (fxComp :type Persistent%20FXComponent%20Object%20Reference))
  (return 0))

(fn initPersistentFX
  (params (fxActor :type Actor%20Object%20Reference) (fxComponent :type Persistent%20FXComponent%20Object%20Reference)))

(fn setFxComponentOwner
  (params (owner :type Persistent%20FXComponent%20Object%20Reference)))

(fn setupFXActor_ServerOnly)
