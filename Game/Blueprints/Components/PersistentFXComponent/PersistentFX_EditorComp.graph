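# Editor-only utility for PersistentFXComponent that spawns a transient copy of the component's chosen FX actor in the editor.
# On init it finds the component on the parent actor, destroys previously spawned actors tagged with its owner's object path, then restarts a short timer.
# The delayed spawn step attaches the new actor to the component, applies that tag and prefixes its label with "Transient_".
# Reuse and no-change checks in the spawn function are left disconnected.
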EditorUtilityBlueprint /Game/Blueprints/Components/PersistentFXComponent/PersistentFX_EditorComp
graphs [
 EventGraph {nodes [
   n1 = Event editorOnly_initEditorComp!() -> (OutputDelegate, then, parent:Object Reference) {is=title K2Node_Event}
   > n2 = Cast To Actor!(execute, Object=@n1.parent) -> (then, CastFailed, AsActor:Actor Object Reference) {is=title K2Node_DynamicCast}
   > n3 = GetComponentByClass!(
    execute
    self:Actor Object Reference=@n2
    ComponentClass:Actor Component Class Reference=&/PersistentFXComponent.PersistentFXComponent_C
   ) -> (then, ReturnValue:Persistent FXComponent Object Reference)
   > Is Valid(InputObject=@n3)
   > editorOnly_SpawnFXActor!(
    execute
    self
    actorClass:Actor Class Reference=Get fxActorToSpawn!(self:Persistent FXComponent Object Reference=@n3) -> (fxActorToSpawn:Actor Class Reference)
    fxComp:Persistent FXComponent Object Reference=@n3
   ) -> (then, actorSpawned:Actor Object Reference)
  ]}
 destroySpawnedActors {nodes [
   destroySpawnedActors() {is=title K2Node_FunctionEntry}
   > n1 = GetAllActorsWithTag!(execute, WorldContextObject:Object Reference, Tag:Name=getActorSpawnTag!() -> (ReturnValue:Name=None)) -> (then, OutActors:Array of Actor Object References)
   > n2 = For Each Loop(Array:Array of Actor Object References=@n1) -> (Array Element:Actor Object Reference)
   > DestroyActor!(execute, self:Editor Actor Subsystem Object Reference=Get EditorActorSubsystem(), ActorToDestroy:Actor Object Reference=@n2) -> (then, ReturnValue:bool)
  ]}
 doTimerSpawn {nodes [
   doTimerSpawn() {is=title K2Node_FunctionEntry}
   > Branch(Condition=IsValidClass!(Class:Object Class Reference=Get currentActorClassToSpawn() -> (currentActorClassToSpawn:Actor Class Reference)) -> (ReturnValue:bool))
   > n1 = SpawnActorFromClass!(execute, ActorClass:Actor Class Reference=Get currentActorClassToSpawn(), Location:Vector="0, 0, 0", Rotation:Rotator, bTransient:bool=true) -> (then, ReturnValue:Actor Object Reference)
   > Attach Actor To Component!(
    execute
    self:Actor Object Reference=@n1
    Parent:Scene Component Object Reference=Get currentTargetFxComp() -> (currentTargetFxComp:Persistent FXComponent Object Reference)
    SocketName:Name=None
    LocationRule:EAttachmentRule Enum=SnapToTarget
    RotationRule:EAttachmentRule Enum=SnapToTarget
    ScaleRule:EAttachmentRule Enum=SnapToTarget
    bWeldSimulatedBodies:bool=false
   ) -> (then, ReturnValue:bool)
   > Set Tags!(
    execute
    Tags:Array of Names=Make Array("[0]":Name=getActorSpawnTag() -> (ReturnValue=None)) -> (Array:Array of Names)
    self:Actor Object Reference=@n1
   ) -> (then, Output_Get:Array of Names)
   > SetActorLabel(self=@n1, NewActorLabel=Append(A=Transient_, B=GetActorLabel(self=@n1, bCreateIfNone=true)), bMarkDirty=true)
  ]}
 editorOnly_SpawnFXActor {nodes [
   n1 = GetAllActorsWithTag(Tag=getActorSpawnTag() -> (ReturnValue=None))
   > Branch(Condition=Is Valid Index!(TargetArray:Array of Actor Object References=@n1, IndexToTest:int) -> (ReturnValue:bool))
   > n2 = For Each Loop(Array=@n1)
   > n3 = Branch(Condition="Equal (Integer)"(A=@n2.Array Index))
   > Attach Actor To Component(
    self=@n2
    Parent=Get currentTargetFxComp()
    SocketName=None
    LocationRule=SnapToTarget
    RotationRule=SnapToTarget
    ScaleRule=SnapToTarget
    bWeldSimulatedBodies=TRUE
   )
   DestroyActor(execute=@n3.else, self=Get EditorActorSubsystem(), ActorToDestroy=@n2)
   n4 = getSpawnedActor!(execute, self) -> (then, currentActorSpawned:Actor Object Reference)
   > n5 = Is Valid(InputObject=@n4)
   PrintString(execute=@n5.Is Not Valid, InString=NO ACTOR EXISTS)
   Branch(execute=@n5.Is Valid, Condition="Equal (Class)"(A=Get currentActorClassToSpawn(), B=Get actorClass() -> (actorClass:Actor Class Reference)))
   > PrintString(InString=No CHANGE)
   > Return Node!(execute, actorSpawned:Actor Object Reference)
   n6 = editorOnly_SpawnFXActor!() -> (then, actorClass:Actor Class Reference, fxComp:Persistent FXComponent Object Reference) {is=title K2Node_FunctionEntry}
   > Set currentActorClassToSpawn(currentActorClassToSpawn:Actor Class Reference=Get fxActorToSpawn(self=@n6.fxComp)) -> (Output_Get:Actor Class Reference)
   > Set currentTargetFxComp(currentTargetFxComp:Persistent FXComponent Object Reference=@n6.fxComp) -> (Output_Get:Persistent FXComponent Object Reference)
   > n7 = Sequence!(execute) -> (then_0, then_1, then_2:Exec, then_3:Exec)
   destroySpawnedActors(execute=@n7.then_2) {is=title K2Node_CallFunction}
   Clear and Invalidate Timer by Handle!(
    execute=@n7.then_3
    WorldContextObject:Object Reference
    Handle=Get timerHandle_SpawnActor() -> (timerHandle_SpawnActor:Timer Handle Structure)
   ) -> (then)
   > n8 = Set Timer by Event(Delegate=Create Event(), Time=0.1, bLooping=false)
   > Set timerHandle_SpawnActor(timerHandle_SpawnActor:Timer Handle Structure=@n8) -> (Output_Get:Timer Handle Structure)
   Return Node()
  ]}
 getActorSpawnTag {nodes [
   getActorSpawnTag() {is=title K2Node_FunctionEntry}
   n1 = GetOwner()
   > Return Node!(execute, ReturnValue:Name=String To Name(InString=Get Object Path String(Object=@n1)))
  ]}
 getSpawnedActor {nodes [
   getSpawnedActor() {is=title K2Node_FunctionEntry}
   n1 = GetAttachedActors!(self:Actor Object Reference=GetOwner(), bResetArray:bool=true, bRecursivelyIncludeAttachedActors:bool=false) -> (OutActors:Array of Actor Object References)
   > n2 = Branch(Condition=Is Valid Index(TargetArray=@n1))
   > Return Node!(
    execute
    currentActorSpawned:Actor Object Reference="Get (a copy)"(Array:Array of Actor Object References=@n1) -> (Output:Actor Object Reference)
   )
   Return Node(execute=@n2.else)
  ]}
]
variables [
 currentActorClassToSpawn {type=Class, default=null}
 currentTargetFxComp {type=PersistentFXComponent_C, default=null}
 timerHandle_SpawnActor {type=TimerHandle}
]
